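# Sound cue for opening the Hightower chest, mixing three layers under the loot reward sound class and a borrowed Galileo box concurrency set.
# The main open sound and a pitch-modulated random treasure chest sweetener each carry their own long-range attenuation override.
# A music layer plays after a short fixed delay, rerouted to the quests sound class, with distance attenuation and spatialization turned off.
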SoundCue /Game/Sounds/Foley_Loot/Containers/HightowerChest/Hightower_Chest_Open
root=Mixer_0
nodes [
 {
  id=Attenuation_0
  children [WavePlayer_2]
  properties {bOverrideAttenuation=true, attenuationOverrides {falloffDistance=100000, stereoSpread=1000, attenuationShapeExtents {x=5000}}}
 }
 {
  id=Attenuation_1
  children [WavePlayer_3]
  properties {
   bOverrideAttenuation=true
   attenuationOverrides {bAttenuate=false, bSpatialize=false, falloffDistance=5000, stereoSpread=512, attenuationShapeExtents {x=10000}}
  }
 }
 {
  id=Attenuation_2
  children [Modulator_3]
  properties {bOverrideAttenuation=true, attenuationOverrides {falloffDistance=100000, stereoSpread=1000, attenuationShapeExtents {x=5000}}}
 }
 {id=Delay_0, children [Attenuation_1], properties {delayMax=0.3, delayMin=0.3}}
 {id=Mixer_0, children [Attenuation_0, Attenuation_2, SoundClass_0], properties {inputVolume [1.1,0.75,2]}}
 {id=Modulator_3, children [Random_6], properties {pitchMax=1.5, pitchMin=1, volumeMax=0.7, volumeMin=0.7}}
 {id=Random_6, children [WavePlayer_16, WavePlayer_17, WavePlayer_18, WavePlayer_19], properties {weights [1,1,1,1]}}
 {id=SoundClass_0, children [Delay_0], properties {soundClassOverride=&Fort_Quests}}
 {id=WavePlayer_16, properties {soundWaveAssetPtr=/Game/Sounds/Foley_Loot/Containers/Treasure_Chest/treasure_chest_sweet_shorter_01}}
 {id=WavePlayer_17, properties {soundWaveAssetPtr=/Game/Sounds/Foley_Loot/Containers/Treasure_Chest/treasure_chest_sweet_shorter_02}}
 {id=WavePlayer_18, properties {soundWaveAssetPtr=/Game/Sounds/Foley_Loot/Containers/Treasure_Chest/treasure_chest_sweet_shorter_03}}
 {id=WavePlayer_19, properties {soundWaveAssetPtr=/Game/Sounds/Foley_Loot/Containers/Treasure_Chest/treasure_chest_sweet_shorter_04}}
 {id=WavePlayer_2, properties {soundWaveAssetPtr=&/Hightower_Chest_Open_01}}
 {id=WavePlayer_3, properties {soundWaveAssetPtr=&/Hightower_Chest_Open_Music_01}}
]
properties {
 attenuationSettings=/Game/Sounds/Attenuation/Fort_Item_Interact_Att
 maxDistance=105000
 soundClassObject=&Fort_Loot_Reward_PSM
 volumeMultiplier=0.8
 attenuationOverrides {falloffDistance=10000, stereoSpread=512, attenuationShapeExtents {x=5000}}
 concurrencySet [/Game/Athena/Sounds/Events/Galileo/Galileo_Box_Conc]
}
